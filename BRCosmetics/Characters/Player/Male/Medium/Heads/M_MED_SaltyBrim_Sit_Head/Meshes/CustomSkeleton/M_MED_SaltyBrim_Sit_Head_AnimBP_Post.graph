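# Post-process Anim Blueprint for the SaltyBrim Sit head that adjusts head and neck and simulates quill physics.
# It shrinks both bones when the local player aims a visible weapon, except certain types, apparently clearing the crosshair.
# An additive rotation counters head-versus-neck_02 motion, reduced to slight yaw during emotes or front end.
# Chained AnimDynamics springs drive eighteen quill bones; computed gravity, RBAN settings and Dynamics values never reach the AnimGraph.
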AnimBlueprint /BRCosmetics/Characters/Player/Male/Medium/Heads/M_MED_SaltyBrim_Sit_Head/Meshes/CustomSkeleton/M_MED_SaltyBrim_Sit_Head_AnimBP_Post
target_skeleton=&/Fortnite_SaltyBrim_Sit_Head_Export_Skel
graphs [
 AnimGraph {nodes [
   n1 = Get scaleDownAlpha() -> (scaleDownAlpha:bool)
   n2 = "Transform (Modify) Bone - Bone: head"!(
    ComponentPose=Local To Component(LocalPose=Input Pose())
    bAlphaBoolEnabled:bool=@n1
    Translation="0,0,0"
    Rotation="0,0,0"
    Scale="0.85,1,1"
   ) -> (Pose) {
    alphaInputType=Bool
    scaleMode=BMM_Replace
    scaleSpace=BCS_BoneSpace
    alphaBoolBlend {blendInTime=0.1, blendOption=Linear, blendOutTime=0.12}
    boneToModify {boneName=head}
    scale {x=0.85}
   }
   n3 = "Transform (Modify) Bone - Bone: neck_02"!(ComponentPose=@n2, bAlphaBoolEnabled:bool=@n1, Translation="0,0,0", Rotation="0,0,0", Scale="1,0.85,0.85") -> (Pose) {
    alphaInputType=Bool
    scaleMode=BMM_Replace
    scaleSpace=BCS_BoneSpace
    alphaBoolBlend {blendInTime=0.1, blendOption=Linear, blendOutTime=0.12}
    boneToModify {boneName=neck_02}
    scale {y=0.85, z=0.85}
   }
   n4 = "Transform (Modify) Bone - Bone: head"!(ComponentPose=@n3, Alpha:float=1, Translation="0,0,0", Rotation=Get head rot() -> (head rot:Rotator), Scale="1,1,1") -> (Pose) {rotationMode=BMM_Additive, rotationSpace=BCS_BoneSpace, alphaBoolBlend {blendOption=Linear}, boneToModify {boneName=head}}
   n5 = Anim Dynamics - Bone: dyn_quill_top_01_c!(ComponentPose=@n4, Alpha=1, GravityOverride:Vector="0,0,-980") -> (Pose) {
    angularDampingOverride=1
    angularSpringConstant=20000
    bAngularSpring=true
    bOverrideAngularDamping=true
    bUseGravityOverride=true
    gravityScale=0.35
    simulationSpace=BoneRelative
    boundBone {boneName=dyn_quill_top_01_c}
    componentAppliedLinearAccClamp=<3000 3000 3000>
    componentLinearAccScale=<1 1 1>
    componentLinearVelScale=<1 1 1>
    gravityOverride {z=-980}
    physicsBodyDefinitions [{
      boundBone {boneName=dyn_quill_top_01_c}
      boxExtents=<15 15 60>
      constraintSetup {angularLimitsMax=<0 10 0>, angularLimitsMin=<0 -5 0>, angularTarget=<1 0 0>}
      localJointOffset=<1 0 0>
     }]
    simSpaceSettings {maxAngularAcceleration=1, maxAngularVelocity=6}
   }
   n6 = Anim Dynamics - Bone: dyn_quill_top_02_c!(ComponentPose=@n5, Alpha=1, GravityOverride:Vector="0,0,-980") -> (Pose) {
    angularDampingOverride=1
    angularSpringConstant=20000
    bAngularSpring=true
    bOverrideAngularDamping=true
    bUseGravityOverride=true
    gravityScale=0.35
    simulationSpace=BoneRelative
    boundBone {boneName=dyn_quill_top_02_c}
    componentAppliedLinearAccClamp=<3000 3000 3000>
    componentLinearAccScale=<1 1 1>
    componentLinearVelScale=<1 1 1>
    gravityOverride {z=-980}
    physicsBodyDefinitions [{
      boundBone {boneName=dyn_quill_top_02_c}
      boxExtents=<15 15 60>
      constraintSetup {angularLimitsMax=<0 10 0>, angularLimitsMin=<0 -5 0>, angularTarget=<1 0 0>}
      localJointOffset=<1 0 0>
     }]
    simSpaceSettings {maxAngularAcceleration=1, maxAngularVelocity=6}
   }
   n7 = Anim Dynamics - Bone: dyn_quill_top_03_c!(ComponentPose=@n6, Alpha=1, GravityOverride:Vector="0,0,-980") -> (Pose) {
    angularDampingOverride=1
    angularSpringConstant=20000
    bAngularSpring=true
    bOverrideAngularDamping=true
    bUseGravityOverride=true
    gravityScale=0.35
    simulationSpace=BoneRelative
    boundBone {boneName=dyn_quill_top_03_c}
    componentAppliedLinearAccClamp=<3000 3000 3000>
    componentLinearAccScale=<1 1 1>
    componentLinearVelScale=<1 1 1>
    gravityOverride {z=-980}
    physicsBodyDefinitions [{
      boundBone {boneName=dyn_quill_top_03_c}
      boxExtents=<15 15 60>
      constraintSetup {angularLimitsMax=<0 10 0>, angularLimitsMin=<0 -5 0>, angularTarget=<1 0 0>}
      localJointOffset=<1 0 0>
     }]
    simSpaceSettings {maxAngularAcceleration=1, maxAngularVelocity=6}
   }
   n8 = Anim Dynamics - Bone: dyn_quill_bottom_01_c!(ComponentPose=@n7, Alpha=1, GravityOverride:Vector="0,0,-980") -> (Pose) {
    angularDampingOverride=1
    angularSpringConstant=20000
    bAngularSpring=true
    bOverrideAngularDamping=true
    bUseGravityOverride=true
    gravityScale=0.35
    simulationSpace=BoneRelative
    boundBone {boneName=dyn_quill_bottom_01_c}
    componentAppliedLinearAccClamp=<3000 3000 3000>
    componentLinearAccScale=<1 1 1>
    componentLinearVelScale=<1 1 1>
    gravityOverride {z=-980}
    physicsBodyDefinitions [{
      boundBone {boneName=dyn_quill_bottom_01_c}
      boxExtents=<15 15 60>
      constraintSetup {angularLimitsMax=<0 10 0>, angularLimitsMin=<0 -5 0>, angularTarget=<1 0 0>}
      localJointOffset=<1 0 0>
     }]
    simSpaceSettings {maxAngularAcceleration=1, maxAngularVelocity=6}
   }
   n9 = Anim Dynamics - Bone: dyn_quill_bottom_02_c!(ComponentPose=@n8, Alpha=1, GravityOverride:Vector="0,0,-980") -> (Pose) {
    angularDampingOverride=1
    angularSpringConstant=20000
    bAngularSpring=true
    bOverrideAngularDamping=true
    bUseGravityOverride=true
    gravityScale=0.35
    simulationSpace=BoneRelative
    boundBone {boneName=dyn_quill_bottom_02_c}
    componentAppliedLinearAccClamp=<3000 3000 3000>
    componentLinearAccScale=<1 1 1>
    componentLinearVelScale=<1 1 1>
    gravityOverride {z=-980}
    physicsBodyDefinitions [{
      boundBone {boneName=dyn_quill_bottom_02_c}
      boxExtents=<15 15 60>
      constraintSetup {angularLimitsMax=<0 10 0>, angularLimitsMin=<0 -5 0>, angularTarget=<1 0 0>}
      localJointOffset=<1 0 0>
     }]
    simSpaceSettings {maxAngularAcceleration=1, maxAngularVelocity=6}
   }
   n10 = Anim Dynamics - Bone: dyn_quill_bottom_03_c!(ComponentPose=@n9, Alpha=1, GravityOverride:Vector="0,0,-980") -> (Pose) {
    angularDampingOverride=1
    angularSpringConstant=20000
    bAngularSpring=true
    bOverrideAngularDamping=true
    bUseGravityOverride=true
    gravityScale=0.35
    simulationSpace=BoneRelative
    boundBone {boneName=dyn_quill_bottom_03_c}
    componentAppliedLinearAccClamp=<3000 3000 3000>
    componentLinearAccScale=<1 1 1>
    componentLinearVelScale=<1 1 1>
    gravityOverride {z=-980}
    physicsBodyDefinitions [{
      boundBone {boneName=dyn_quill_bottom_03_c}
      boxExtents=<15 15 60>
      constraintSetup {angularLimitsMax=<0 10 0>, angularLimitsMin=<0 -5 0>, angularTarget=<1 0 0>}
      localJointOffset=<1 0 0>
     }]
    simSpaceSettings {maxAngularAcceleration=1, maxAngularVelocity=6}
   }
   n11 = Anim Dynamics - Bone: dyn_quill_top_01_l!(ComponentPose=@n10, Alpha=1, GravityOverride:Vector="0,0,-980") -> (Pose) {
    angularDampingOverride=1
    angularSpringConstant=20000
    bAngularSpring=true
    bOverrideAngularDamping=true
    bUseGravityOverride=true
    gravityScale=0.35
    simulationSpace=BoneRelative
    boundBone {boneName=dyn_quill_top_01_l}
    componentAppliedLinearAccClamp=<3000 3000 3000>
    componentLinearAccScale=<1 1 1>
    componentLinearVelScale=<1 1 1>
    gravityOverride {z=-980}
    physicsBodyDefinitions [{
      boundBone {boneName=dyn_quill_top_01_l}
      boxExtents=<15 15 60>
      constraintSetup {angularLimitsMax=<0 10 0>, angularLimitsMin=<0 -5 0>, angularTarget=<1 0 0>}
      localJointOffset=<1 0 0>
     }]
    simSpaceSettings {maxAngularAcceleration=1, maxAngularVelocity=6}
   }
   n12 = Anim Dynamics - Bone: dyn_quill_top_02_l!(ComponentPose=@n11, Alpha=1, GravityOverride:Vector="0,0,-980") -> (Pose) {
    angularDampingOverride=1
    angularSpringConstant=20000
    bAngularSpring=true
    bOverrideAngularDamping=true
    bUseGravityOverride=true
    gravityScale=0.35
    simulationSpace=BoneRelative
    boundBone {boneName=dyn_quill_top_02_l}
    componentAppliedLinearAccClamp=<3000 3000 3000>
    componentLinearAccScale=<1 1 1>
    componentLinearVelScale=<1 1 1>
    gravityOverride {z=-980}
    physicsBodyDefinitions [{
      boundBone {boneName=dyn_quill_top_02_l}
      boxExtents=<15 15 60>
      constraintSetup {angularLimitsMax=<0 10 0>, angularLimitsMin=<0 -5 0>, angularTarget=<1 0 0>}
      localJointOffset=<1 0 0>
     }]
    simSpaceSettings {maxAngularAcceleration=1, maxAngularVelocity=6}
   }
   n13 = Anim Dynamics - Bone: dyn_quill_top_03_l!(ComponentPose=@n12, Alpha=1, GravityOverride:Vector="0,0,-980") -> (Pose) {
    angularDampingOverride=1
    angularSpringConstant=20000
    bAngularSpring=true
    bOverrideAngularDamping=true
    bUseGravityOverride=true
    gravityScale=0.35
    simulationSpace=BoneRelative
    boundBone {boneName=dyn_quill_top_03_l}
    componentAppliedLinearAccClamp=<3000 3000 3000>
    componentLinearAccScale=<1 1 1>
    componentLinearVelScale=<1 1 1>
    gravityOverride {z=-980}
    physicsBodyDefinitions [{
      boundBone {boneName=dyn_quill_top_03_l}
      boxExtents=<15 15 60>
      constraintSetup {angularLimitsMax=<0 10 0>, angularLimitsMin=<0 -5 0>, angularTarget=<1 0 0>}
      localJointOffset=<1 0 0>
     }]
    simSpaceSettings {maxAngularAcceleration=1, maxAngularVelocity=6}
   }
   n14 = Anim Dynamics - Bone: dyn_quill_bottom_01_l!(ComponentPose=@n13, Alpha=1, GravityOverride:Vector="0,0,-980") -> (Pose) {
    angularDampingOverride=1
    angularSpringConstant=20000
    bAngularSpring=true
    bOverrideAngularDamping=true
    bUseGravityOverride=true
    gravityScale=0.35
    simulationSpace=BoneRelative
    boundBone {boneName=dyn_quill_bottom_01_l}
    componentAppliedLinearAccClamp=<3000 3000 3000>
    componentLinearAccScale=<1 1 1>
    componentLinearVelScale=<1 1 1>
    gravityOverride {z=-980}
    physicsBodyDefinitions [{
      boundBone {boneName=dyn_quill_bottom_01_l}
      boxExtents=<15 15 60>
      constraintSetup {angularLimitsMax=<0 10 0>, angularLimitsMin=<0 -5 0>, angularTarget=<1 0 0>}
      localJointOffset=<1 0 0>
     }]
    simSpaceSettings {maxAngularAcceleration=1, maxAngularVelocity=6}
   }
   n15 = Anim Dynamics - Bone: dyn_quill_bottom_02_l!(ComponentPose=@n14, Alpha=1, GravityOverride:Vector="0,0,-980") -> (Pose) {
    angularDampingOverride=1
    angularSpringConstant=20000
    bAngularSpring=true
    bOverrideAngularDamping=true
    bUseGravityOverride=true
    gravityScale=0.35
    simulationSpace=BoneRelative
    boundBone {boneName=dyn_quill_bottom_02_l}
    componentAppliedLinearAccClamp=<3000 3000 3000>
    componentLinearAccScale=<1 1 1>
    componentLinearVelScale=<1 1 1>
    gravityOverride {z=-980}
    physicsBodyDefinitions [{
      boundBone {boneName=dyn_quill_bottom_02_l}
      boxExtents=<15 15 60>
      constraintSetup {angularLimitsMax=<0 10 0>, angularLimitsMin=<0 -5 0>, angularTarget=<1 0 0>}
      localJointOffset=<1 0 0>
     }]
    simSpaceSettings {maxAngularAcceleration=1, maxAngularVelocity=6}
   }
   n16 = Anim Dynamics - Bone: dyn_quill_bottom_03_l!(ComponentPose=@n15, Alpha=1, GravityOverride:Vector="0,0,-980") -> (Pose) {
    angularDampingOverride=1
    angularSpringConstant=20000
    bAngularSpring=true
    bOverrideAngularDamping=true
    bUseGravityOverride=true
    gravityScale=0.35
    simulationSpace=BoneRelative
    boundBone {boneName=dyn_quill_bottom_03_l}
    componentAppliedLinearAccClamp=<3000 3000 3000>
    componentLinearAccScale=<1 1 1>
    componentLinearVelScale=<1 1 1>
    gravityOverride {z=-980}
    physicsBodyDefinitions [{
      boundBone {boneName=dyn_quill_bottom_03_l}
      boxExtents=<15 15 60>
      constraintSetup {angularLimitsMax=<0 10 0>, angularLimitsMin=<0 -5 0>, angularTarget=<1 0 0>}
      localJointOffset=<1 0 0>
     }]
    simSpaceSettings {maxAngularAcceleration=1, maxAngularVelocity=6}
   }
   n17 = Anim Dynamics - Bone: dyn_quill_top_01_r!(ComponentPose=@n16, Alpha=1, GravityOverride:Vector="0,0,-980") -> (Pose) {
    angularDampingOverride=1
    angularSpringConstant=20000
    bAngularSpring=true
    bOverrideAngularDamping=true
    bUseGravityOverride=true
    gravityScale=0.35
    simulationSpace=BoneRelative
    boundBone {boneName=dyn_quill_top_01_r}
    componentAppliedLinearAccClamp=<3000 3000 3000>
    componentLinearAccScale=<1 1 1>
    componentLinearVelScale=<1 1 1>
    gravityOverride {z=-980}
    physicsBodyDefinitions [{
      boundBone {boneName=dyn_quill_top_01_r}
      boxExtents=<15 15 60>
      constraintSetup {angularLimitsMax=<0 10 0>, angularLimitsMin=<0 -5 0>, angularTarget=<1 0 0>}
      localJointOffset=<1 0 0>
     }]
    simSpaceSettings {maxAngularAcceleration=1, maxAngularVelocity=6}
   }
   n18 = Anim Dynamics - Bone: dyn_quill_top_02_r!(ComponentPose=@n17, Alpha=1, GravityOverride:Vector="0,0,-980") -> (Pose) {
    angularDampingOverride=1
    angularSpringConstant=20000
    bAngularSpring=true
    bOverrideAngularDamping=true
    bUseGravityOverride=true
    gravityScale=0.35
    simulationSpace=BoneRelative
    boundBone {boneName=dyn_quill_top_02_r}
    componentAppliedLinearAccClamp=<3000 3000 3000>
    componentLinearAccScale=<1 1 1>
    componentLinearVelScale=<1 1 1>
    gravityOverride {z=-980}
    physicsBodyDefinitions [{
      boundBone {boneName=dyn_quill_top_02_r}
      boxExtents=<15 15 60>
      constraintSetup {angularLimitsMax=<0 10 0>, angularLimitsMin=<0 -5 0>, angularTarget=<1 0 0>}
      localJointOffset=<1 0 0>
     }]
    simSpaceSettings {maxAngularAcceleration=1, maxAngularVelocity=6}
   }
   n19 = Anim Dynamics - Bone: dyn_quill_top_03_r!(ComponentPose=@n18, Alpha=1, GravityOverride:Vector="0,0,-980") -> (Pose) {
    angularDampingOverride=1
    angularSpringConstant=20000
    bAngularSpring=true
    bOverrideAngularDamping=true
    bUseGravityOverride=true
    gravityScale=0.35
    simulationSpace=BoneRelative
    boundBone {boneName=dyn_quill_top_03_r}
    componentAppliedLinearAccClamp=<3000 3000 3000>
    componentLinearAccScale=<1 1 1>
    componentLinearVelScale=<1 1 1>
    gravityOverride {z=-980}
    physicsBodyDefinitions [{
      boundBone {boneName=dyn_quill_top_03_r}
      boxExtents=<15 15 60>
      constraintSetup {angularLimitsMax=<0 10 0>, angularLimitsMin=<0 -5 0>, angularTarget=<1 0 0>}
      localJointOffset=<1 0 0>
     }]
    simSpaceSettings {maxAngularAcceleration=1, maxAngularVelocity=6}
   }
   n20 = Anim Dynamics - Bone: dyn_quill_bottom_01_r!(ComponentPose=@n19, Alpha=1, GravityOverride:Vector="0,0,-980") -> (Pose) {
    angularDampingOverride=1
    angularSpringConstant=20000
    bAngularSpring=true
    bOverrideAngularDamping=true
    bUseGravityOverride=true
    gravityScale=0.35
    simulationSpace=BoneRelative
    boundBone {boneName=dyn_quill_bottom_01_r}
    componentAppliedLinearAccClamp=<3000 3000 3000>
    componentLinearAccScale=<1 1 1>
    componentLinearVelScale=<1 1 1>
    gravityOverride {z=-980}
    physicsBodyDefinitions [{
      boundBone {boneName=dyn_quill_bottom_01_r}
      boxExtents=<15 15 60>
      constraintSetup {angularLimitsMax=<0 10 0>, angularLimitsMin=<0 -5 0>, angularTarget=<1 0 0>}
      localJointOffset=<1 0 0>
     }]
    simSpaceSettings {maxAngularAcceleration=1, maxAngularVelocity=6}
   }
   n21 = Anim Dynamics - Bone: dyn_quill_bottom_02_r!(ComponentPose=@n20, Alpha=1, GravityOverride:Vector="0,0,-980") -> (Pose) {
    angularDampingOverride=1
    angularSpringConstant=20000
    bAngularSpring=true
    bOverrideAngularDamping=true
    bUseGravityOverride=true
    gravityScale=0.35
    simulationSpace=BoneRelative
    boundBone {boneName=dyn_quill_bottom_02_r}
    componentAppliedLinearAccClamp=<3000 3000 3000>
    componentLinearAccScale=<1 1 1>
    componentLinearVelScale=<1 1 1>
    gravityOverride {z=-980}
    physicsBodyDefinitions [{
      boundBone {boneName=dyn_quill_bottom_02_r}
      boxExtents=<15 15 60>
      constraintSetup {angularLimitsMax=<0 10 0>, angularLimitsMin=<0 -5 0>, angularTarget=<1 0 0>}
      localJointOffset=<1 0 0>
     }]
    simSpaceSettings {maxAngularAcceleration=1, maxAngularVelocity=6}
   }
   n22 = Anim Dynamics - Bone: dyn_quill_bottom_03_r!(ComponentPose=@n21, Alpha=1, GravityOverride:Vector="0,0,-980") -> (Pose) {
    angularDampingOverride=1
    angularSpringConstant=20000
    bAngularSpring=true
    bOverrideAngularDamping=true
    bUseGravityOverride=true
    gravityScale=0.35
    simulationSpace=BoneRelative
    boundBone {boneName=dyn_quill_bottom_03_r}
    componentAppliedLinearAccClamp=<3000 3000 3000>
    componentLinearAccScale=<1 1 1>
    componentLinearVelScale=<1 1 1>
    gravityOverride {z=-980}
    physicsBodyDefinitions [{
      boundBone {boneName=dyn_quill_bottom_03_r}
      boxExtents=<15 15 60>
      constraintSetup {angularLimitsMax=<0 10 0>, angularLimitsMin=<0 -5 0>, angularTarget=<1 0 0>}
      localJointOffset=<1 0 0>
     }]
    simSpaceSettings {maxAngularAcceleration=1, maxAngularVelocity=6}
   }
   Output Pose(Result=Component To Local(ComponentPose=@n22))
  ]}
 EventGraph {nodes [
   Event BlueprintUpdateAnimation()
   > n1 = Sequence!(execute) -> (then_0, then_1, then_2:Exec, then_3:Exec)
   > n2 = Branch(Condition=OR Boolean(A=Get bIsPlayingEmote(), B=Get bIsInFrontEnd()))
   n3 = GetOwnerSkeletalMeshComponent()
   n4 = BreakRotator(InRot=BreakTransform(InTransform=GetDeltaTransformFromRefPose(self=@n3, BoneName=head, BaseName=neck_02)).Rotation)
   n5 = Get Z percentage() -> (Z percentage:float)
   n6 = Get Y percentage() -> (Y percentage:float)
   n7 = Get X percentage() -> (X percentage:float)
   > n8 = Set head rot(head rot:Rotator=MakeRotator(
     Roll=float * float!(A=@n4, B=0, C:double=@n7) -> (ReturnValue)
     Pitch=float * float(A=@n4.Pitch, B=0, C=@n6)
     Yaw=float * float(A=@n4.Yaw, B=-0.3, C=@n5)
    )) -> (Output_Get:Rotator)
   n9 = "Gravity Override State Function (Parameter Asset) (Nativized)"(
    execute=[@n8.then, @n15.then]
    CharacterPartAnimInstance=Self-Reference()
    GameTimeInSeconds=GetGameTimeInSeconds()
    Parameters=/BRCosmetics/Characters/Player/Male/Medium/Heads/M_MED_SaltyBrim_Sit_Head/Meshes/M_MED_SaltyBrim_Sit_Parameters
   )
   > Set GravityOverride(GravityOverride=@n9)
   > Set RBAN Settings Out(RBAN Settings Out=@n9.OutRigidBodyAnimNodeParameters)
   n10 = GetOwnerSkeletalMeshComponent()
   n11 = BreakRotator(InRot=BreakTransform(InTransform=GetDeltaTransformFromRefPose(self=@n10, BoneName=head, BaseName=neck_02)).Rotation)
   n12 = Get Z percentage()
   n13 = Get X percentage()
   n14 = Get Y percentage()
   n15 = Set head rot(
    execute=@n2.else
    head rot=MakeRotator(Roll=float * float(A=@n11, B=-1.5, C=@n13), Pitch=float * float(A=@n11.Pitch, B=-1.5, C=@n14), Yaw=float * float(A=@n11.Yaw, B=-1.5, C=@n12))
   )
   Set FortPawn(execute=@n1.then_1, FortPawn:Fort Pawn Object Reference=TryGetFortPawn()) -> (Output_Get:Fort Pawn Object Reference)
   n16 = Get FortPawn() -> (FortPawn:Fort Pawn Object Reference)
   > n17 = Is Valid(InputObject=@n16)
   Set weapon_visible(execute=@n17.Is Not Valid, weapon_visible:bool=false) -> (Output_Get:bool)
   n18 = Get WeaponCoreAnim!() -> (WeaponCoreAnim)
   n19 = Set notCoverCrosshair(
    execute=@n17.Is Valid
    notCoverCrosshair:bool=OR Boolean!(
     A="Equal (Enum)"(A:EFortWeaponCoreAnimation Enum=@n18, B:EFortWeaponCoreAnimation Enum=TrapPlacement)
     B="Equal (Enum)"(A=@n18, B=RocketLauncher)
     C:bool="Equal (Enum)"(A=@n18, B=GoingCommando)
    ) -> (ReturnValue)
   ) -> (Output_Get:bool)
   > n20 = Branch(Condition=@n19)
   > Set weapon_visible(weapon_visible=false)
   n21 = IsWeaponHolstered!(self:Fort Pawn Object Reference=@n16) -> (ReturnValue:bool)
   n22 = Get CurrentWeapon() -> (CurrentWeapon:Fort Weapon Object Reference)
   Set weapon_visible(execute=@n20.else, weapon_visible=AND Boolean(A=IsValid(Object=@n22), B=NOT Boolean(A=@n21)))
   n23 = Branch(execute=@n1.then_3, Condition=Get bIsPlayingEmote())
   > Set Dynamics(Dynamics:double=0.25) -> (Output_Get:double)
   Set Dynamics(execute=@n23.else, Dynamics=1)
   n24 = Set isLocalPlayer(execute=@n1.then_2, isLocalPlayer:bool=IsLocalPawn!(Actor:Actor Object Reference=GetOwningActor()) -> (IsLocalPawn:bool)) -> (Output_Get:bool)
   > n25 = Branch(Condition=AND Boolean!(A=@n24, B=Get bIsTargeting(), C:bool=Get weapon_visible() -> (weapon_visible:bool)) -> (ReturnValue))
   > Set scaleDownAlpha(scaleDownAlpha:bool=true) -> (Output_Get:bool)
   Set scaleDownAlpha(execute=@n25.else, scaleDownAlpha=false)
   TryGetPawnOwner()
  ]}
]
variables [
 Dynamics {type=number}
 FortPawn {type=FortPawn, default=null}
 GravityOverride {type=Vector}
 head rot {type=Rotator}
 isLocalPlayer {type=boolean}
 notCoverCrosshair {type=boolean}
 RBAN Settings Out {type=FortRigidBodyAnimNodeParameters}
 scaleDownAlpha {type=boolean}
 weapon_visible {type=boolean}
 X percentage {type=number, default=0.4}
 Y percentage {type=number, default=0.4}
 Z percentage {type=number, default=0.8}
]
